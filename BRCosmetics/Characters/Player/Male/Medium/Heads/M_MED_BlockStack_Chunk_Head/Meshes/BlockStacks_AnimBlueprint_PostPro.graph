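# Post-process Anim Blueprint for the BlockStack Chunk male player head, passing the input pose through a Control Rig driven by character state.
# Each update converts crouch, movement, idle and jump/fall states into floats; both Moving2D values use the same crouching-and-moving test and feed oppositely named rig pins.
# Weapon changes reset, then set flags for assault rifle, pistol, P90 SMG or tactical shotgun, feeding the rig alongside targeting, melee, emote and skydive states.
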AnimBlueprint /BRCosmetics/Characters/Player/Male/Medium/Heads/M_MED_BlockStack_Chunk_Head/Meshes/BlockStacks_AnimBlueprint_PostPro
target_skeleton=&/M_MED_BlockStacks_Skeleton_Custom
graphs [
 AnimGraph {nodes [
   n1 = Control Rig!(
    Source=Input Pose()
    Alpha
    isTargeting lerp:double=0
    isAR Equipped lerp:double=0
    pub isTargeting:bool=Get bIsTargeting()
    pub isAR Equipped:bool=Get isAR Equipped() -> (isAR Equipped:bool)
    is_OutLaw_SG:bool=Get Is_outLaw_SG() -> (Is_outLaw_SG:bool)
    SMG_P90:bool=Get SMG_SG() -> (SMG_SG:bool)
    pub isPistol Equipped:bool=Get isPistol Equipped() -> (isPistol Equipped:bool)
    pub PA Swing:bool=Get bIsPlayingMeleeAnim()
    CrouchingBool:bool=Get bIsCrouching()
    Crouching:double=0
    StandingMovng2D:double=Get CrouchingMoving2D() -> (CrouchingMoving2D:double)
    CrouchingMoving2D:double=Get StandingMoving2D() -> (StandingMoving2D:double)
    Idle:double=Get Idle() -> (Idle:double)
    JumpingComplete:double=Get JumpingComplete() -> (JumpingComplete:double)
    EmoteBool:bool=Get bIsPlayingEmote()
    Skydiving:bool=Get bIsSkyDiving()
   ) -> (Pose) {controlRigAssetReference {controlRigAsset=&/M_MED_BlockStack_Chunk_Head_CtrlRig1}}
   Output Pose(Result=@n1)
  ]}
 EventGraph {nodes [
   Event BlueprintUpdateAnimation()
   n1 = Get bIsCrouching()
   n2 = Get bIsMoving2D()
   > Set CrouchingMoving2D(CrouchingMoving2D:double="To Float (Boolean)"(InBool=AND Boolean(A=@n1, B=@n2))) -> (Output_Get:double)
   n3 = Get bIsMoving2D()
   n4 = Get bIsCrouching()
   > Set StandingMoving2D(StandingMoving2D:double="To Float (Boolean)"(InBool=AND Boolean(A=@n3, B=@n4))) -> (Output_Get:double)
   n5 = NOT Boolean(A=Get bIsPlayingEmote())
   n6 = NOT Boolean(A=Get bIsMoving2D())
   n7 = NOT Boolean(A=Get bIsSkyDiving())
   > Set Idle(Idle:double="To Float (Boolean)"(InBool=AND Boolean!(A=@n6, B=@n7, C:bool=@n5) -> (ReturnValue))) -> (Output_Get:double)
   n8 = AND Boolean!(A=NOT Boolean(A=Get bIsSkyDiving()), B=Get bIsFalling()) -> (ReturnValue)
   n9 = Get bIsJumping()
   > Set JumpingComplete(JumpingComplete:double="To Float (Boolean)"(InBool=OR Boolean(A=@n8, B=@n9))) -> (Output_Get:double)
   Event OnWeaponChanged()
   > Set isAR Equipped(isAR Equipped:bool=false) -> (Output_Get:bool)
   > Set isPistol Equipped(isPistol Equipped:bool=false) -> (Output_Get:bool)
   > Set Is_outLaw_SG(Is_outLaw_SG:bool=false) -> (Output_Get:bool)
   > Set SMG_SG(SMG_SG:bool=false) -> (Output_Get:bool)
   > n10 = Switch on EFortWeaponCoreAnimation!(execute, Selection:EFortWeaponCoreAnimation Enum=Get WeaponCoreAnim!() -> (WeaponCoreAnim)) -> (
    Melee:Exec
    Pistol:Exec
    Shotgun:Exec
    PaperBlueprint:Exec
    Rifle:Exec
    MeleeOneHand:Exec
    MachinePistol:Exec
    RocketLauncher:Exec
    GrenadeLauncher:Exec
    GoingCommando:Exec
    AssaultRifle:Exec
    TacticalShotgun:Exec
    SniperRifle:Exec
    TrapPlacement:Exec
    ShoulderLauncher:Exec
    AbilityDecoTool:Exec
    Crossbow:Exec
    C4:Exec
    RemoteControl:Exec
    DualWield:Exec
    AR_BullPup:Exec
    AR_ForwardGrip:Exec
    MedPackPaddles:Exec
    SMG_P90:Exec
    AR_DrumGun:Exec
    Consumable_Small:Exec
    Consumable_Large:Exec
    Balloon:Exec
    MountedTurret:Exec
    CreativeTool:Exec
    ExplosiveBow:Exec
    AshtonIndigo:Exec
    AshtonChicago:Exec
    MeleeDualWield:Exec
    Unarmed:Exec
   )
   Set isAR Equipped(execute=@n10.AssaultRifle, isAR Equipped=true)
   Set isPistol Equipped(execute=@n10.Pistol, isPistol Equipped=true)
   Set SMG_SG(execute=@n10.SMG_P90, SMG_SG=true)
   Set Is_outLaw_SG(execute=@n10.TacticalShotgun, Is_outLaw_SG=true)
   TryGetPawnOwner()
  ]}
]
variables [
 CrouchingMoving2D {type=number}
 Idle {type=number}
 Is_outLaw_SG {type=boolean}
 isAR Equipped {type=boolean}
 isPistol Equipped {type=boolean}
 JumpingComplete {type=number}
 SMG_SG {type=boolean}
 StandingMoving2D {type=number}
]
